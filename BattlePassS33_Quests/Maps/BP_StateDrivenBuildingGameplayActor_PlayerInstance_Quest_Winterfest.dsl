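; Player-instance state-driven building actor for a Winterfest quest that toggles the Noir quest day sequence look as quest state changes.
; On non-authority clients, side-effect tags enable the look for State B and disable it for A or Success.
; A timeline played forward or in reverse fades the child volume's DaySequence modifier blend weight.
; On BeginPlay, a flag can destroy the volume component entirely.
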(event EventBeginPlay
  (|CallParentFunction)
  (if (Variables|Default|GetDestroy%20DaySeqVolume)
    (Components|DestroyComponent (Variables|Default|GetB_DaySequenceVolume_NoirQuest))))

(event SideEffects|EventBP_ApplySideEffect
  (params (ReactionSource :type Object%20Reference) (TargetObject :type Object%20Reference) (SideEffectTriggerTag :type Gameplay%20Tag%20Structure%20%28by%20ref%29) (SideEffectTriggerType :type EBlueprintableSideEffectTriggerType%20Enum%20%28by%20ref%29 :default "TriggeredAsTarget"))
  (Networking|SwitchHasAuthority
    (:Remote
      (switch Utilities|FlowControl|Switch|SwitchonGameplayTag SideEffectTriggerTag
        (:Quests.SDA.BPReaction.State.A
          _)
        (:Quests.SDA.BPReaction.State.B
          (CallFunction|EnableModifier))
        (:Quests.SDA.BPReaction.State.Success
          _))
      (CallFunction|DisableModifier))))

(event Custom|Enable%20Modifier
  (bind _b_daysequencevolume_noirquest (Variables|Default|GetB_DaySequenceVolume_NoirQuest))
  (bind _childactor (Variables|/Script/Engine.ChildActorComponent|GetChildActor _b_daysequencevolume_noirquest))
  (bind _asb_day_sequence_volume_noir_quest (Utilities|Casting|CastToB_DaySequenceVolume_NoirQuest _childactor))
  (bind _daysequencemodifier (Variables|/Game/TimeOfDay/TODM_Sequence/Blueprint/Volume/B_DaySequenceVolume_Parent.B_DaySequenceVolume_Parent_C|GetDaySequenceModifier _asb_day_sequence_volume_noir_quest))
  (Utilities|IsValid _daysequencemodifier
    (:Is%20Valid
      (bind (_direction _blend) (Timeline|Timeline PlayFromStart :NewTime 1.0)
        (:Update
          ; 0 for off, 1 for on
          (Variables|/Script/DaySequence.DaySequenceModifierComponent|SetUserBlendWeight _blend _daysequencemodifier))))))

(event Custom|Disable%20Modifier
  (bind _b_daysequencevolume_noirquest (Variables|Default|GetB_DaySequenceVolume_NoirQuest))
  (bind _childactor (Variables|/Script/Engine.ChildActorComponent|GetChildActor _b_daysequencevolume_noirquest))
  (bind _asb_day_sequence_volume_noir_quest (Utilities|Casting|CastToB_DaySequenceVolume_NoirQuest _childactor))
  (bind _daysequencemodifier (Variables|/Game/TimeOfDay/TODM_Sequence/Blueprint/Volume/B_DaySequenceVolume_Parent.B_DaySequenceVolume_Parent_C|GetDaySequenceModifier _asb_day_sequence_volume_noir_quest))
  (Utilities|IsValid _daysequencemodifier
    (:Is%20Valid
      (Timeline|Timeline ReverseFromEnd))))

(fn ConstructionScript
  (|CallParentFunction))
